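# Post-sim Animation Blueprint for the MagmaBreak head cosmetic that animates its two base shard bones.
# EventGraph sets a time-based sine X offset per shard and converts bIsTargeting to an alpha.
# The offsets are added to the input pose, cached, then fed through partially weighted, zero-gravity, root-relative AnimDynamics springs.
# A final targeting-weighted additive transform shifts and rotates the left shard; the right shard's matching transform is zero.
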AnimBlueprint /BRCosmetics/Characters/Player/Male/Medium/Heads/M_MED_MagmaBreak_Head/Meshes/CustomSkeleton/Head_MagmaBreak_PostSim_AnimBP
target_skeleton=&/Fortnite_Base_Head_MagmaBreak_Skeleton
graphs [
 AnimGraph {nodes [
   n1 = Anim Dynamics - Bone: base_shard_l(ComponentPose=Local To Component(LocalPose=Use cached pose 'TranslatePose'()), Alpha=0.35, ExternalForce="0,0,0") {
    angularDampingOverride=0.9
    angularSpringConstant=200
    bGravityOverrideInSimSpace=true
    bLinearSpring=true
    bOverrideAngularDamping=true
    bOverrideLinearDamping=true
    gravityScale=0
    linearDampingOverride=0.9999
    linearSpringConstant=100
    simulationSpace=RootRelative
    alphaBoolBlend {blendOption=Linear}
    boundBone {boneName=base_shard_l}
    componentAppliedLinearAccClamp=<50000 50000 50000>
    componentLinearAccScale {x=0.1, z=0.1}
    componentLinearVelScale=<0.01 0.01 0.01>
    physicsBodyDefinitions [{
      sphereCollisionRadius=0
      boundBone {boneName=base_shard_l}
      boxExtents=<4 4 4>
      constraintSetup {angularLimitsMax=<0 0 0>, angularLimitsMin=<0 0 0>, angularTarget=<0 0 0>, linearAxesMax=<4.3 4.3 4.3>, linearAxesMin=<-0.5 -0.1 -0.5>}
      localJointOffset=<0 0 0>
     }]
    relativeSpaceBone {boneName=root}
   }
   n2 = Anim Dynamics - Bone: base_shard_r(ComponentPose=@n1, Alpha=0.35, ExternalForce="0,0,0") {
    angularDampingOverride=0.9
    angularSpringConstant=200
    bGravityOverrideInSimSpace=true
    bLinearSpring=true
    bOverrideAngularDamping=true
    bOverrideLinearDamping=true
    gravityScale=0
    linearDampingOverride=0.9999
    linearSpringConstant=100
    simulationSpace=RootRelative
    alphaBoolBlend {blendOption=Linear}
    boundBone {boneName=base_shard_r}
    componentAppliedLinearAccClamp=<50000 50000 50000>
    componentLinearAccScale {x=0.1, z=0.1}
    componentLinearVelScale=<0.01 0.01 0.01>
    physicsBodyDefinitions [{
      sphereCollisionRadius=0
      boundBone {boneName=base_shard_r}
      boxExtents=<4 4 4>
      constraintSetup {angularLimitsMax=<0 0 0>, angularLimitsMin=<0 0 0>, angularTarget=<0 0 0>, linearAxesMax=<4.3 4.3 4.3>, linearAxesMin=<-0.5 -0.1 -0.5>}
      localJointOffset=<0 0 0>
     }]
    relativeSpaceBone {boneName=root}
   }
   n3 = Get Is Targeting Alpha() -> (Is Targeting Alpha:double)
   n4 = "Transform (Modify) Bone - Bone: base_shard_r"(ComponentPose=@n2, Alpha=@n3, Translation="0,0,0", Rotation="0,0,0", Scale="1,1,1") {translationMode=BMM_Additive, translationSpace=BCS_BoneSpace, alphaBoolBlend {blendOption=Linear}, boneToModify {boneName=base_shard_r}}
   n5 = "Transform (Modify) Bone - Bone: base_shard_l"(ComponentPose=@n4, Alpha=@n3, Translation="2,0,2", Rotation="28.022,-22.792,-11.185", Scale="1,1,1") {
    rotationMode=BMM_Additive
    rotationSpace=BCS_BoneSpace
    translationMode=BMM_Additive
    translationSpace=BCS_BoneSpace
    alphaBoolBlend {blendOption=Linear}
    alphaScaleBiasClamp {bInterpResult=true}
    boneToModify {boneName=base_shard_l}
    rotation=<rot 28.022 -22.792 -11.185>
    translation {x=2, z=2}
   }
   Output Pose(Result=Component To Local(ComponentPose=@n5))
   n6 = "Transform (Modify) Bone - Bone: base_shard_r"(
    ComponentPose=Local To Component(LocalPose=Input Pose())
    Alpha=1
    Translation=Get TranslateShard_L() -> (TranslateShard_L:Vector)
    Rotation="0,0,0"
    Scale="1,1,1"
   ) {translationMode=BMM_Additive, translationSpace=BCS_BoneSpace, alphaBoolBlend {blendOption=Linear}, boneToModify {boneName=base_shard_r}}
   n7 = "Transform (Modify) Bone - Bone: base_shard_l"(ComponentPose=@n6, Alpha=1, Translation=Get TranslateShard_R() -> (TranslateShard_R:Vector), Rotation="0,0,0", Scale="1,1,1") {translationMode=BMM_Additive, translationSpace=BCS_BoneSpace, alphaBoolBlend {blendOption=Linear}, boneToModify {boneName=base_shard_l}}
   Save cached pose 'TranslatePose'(Pose=Component To Local(ComponentPose=@n7)) {cacheName=TranslatePose}
  ]}
 EventGraph {nodes [
   Event BlueprintUpdateAnimation()
   > n1 = Sequence!(execute) -> (then_0, then_1, then_2:Exec)
   n2 = "Sin (Radians)"(A=float * float(A=GetRealTimeSeconds(), B=4))
   > Set TranslateShard_R(TranslateShard_R:Vector=MakeVector(X=float / float(A=@n2, B=0.5) {is=title K2Node_CallFunction}, Y=0)) -> (Output_Get:Vector)
   n3 = "Sin (Radians)"(A=float * float(A=GetRealTimeSeconds(), B=4))
   Set TranslateShard_L(execute=@n1.then_1, TranslateShard_L:Vector=MakeVector(X=float / float(A=@n3, B=0.5), Y=0)) -> (Output_Get:Vector)
   Set Is Targeting Alpha(execute=@n1.then_2, Is Targeting Alpha:double="To Float (Boolean)"(InBool=Get bIsTargeting())) -> (Output_Get:double)
   TryGetPawnOwner()
  ]}
]
variables [
 Is Targeting Alpha {type=number}
 TranslateShard_L {type=Vector, default=<-0.8546 0 0>}
 TranslateShard_R {type=Vector, default=<-0.6574 0 0>}
]
